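# Control Rig for the NoiseClueGust pickaxe that procedurally wiggles its four tentacle bone chains (A to D).
# BeginExecution runs eight ChainHarmonicsPerItem units in series, two per chain: an X-axis wave, then a Y-axis wave with Reach.
# Differing wave offsets stagger motion between chains; Pendulum is disabled throughout.
# Debug draw is set on every unit, and two extra tentacle A units sit outside the execution chain.
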ControlRigBlueprint /BRCosmetics/Weapons/FORT_Melee/Pickaxe_NoiseClueGust/Meshes/NoiseClueGust_Axe_CtrlRig
graphs [
 RigVMFunctionLibrary {links [], nodes []}
 RigVMModel {nodes [
   RigUnit_BeginExecution = &RigUnit_BeginExecution()
   /Script/ControlRig.RigUnit_ChainHarmonicsPerItem(
    ChainRoot=(Type=Bone,Name="dyn_tentacle_A_base")
    DrawWorldOffset=(Rotation=(X=0,Y=0,Z=0,W=1),Translation=(X=0,Y=0,Z=0),Scale3D=(X=1,Y=1,Z=1))
    Pendulum=(bEnabled=false,PendulumStiffness=50,PendulumGravity=(X=-900,Y=-500,Z=0),PendulumBlend=1,PendulumDrag=0.98,PendulumMinimum=0,PendulumMaximum=0.9,PendulumEase=SineEaseInOut,UnwindAxis=(X=0,Y=0,Z=0),UnwindMinimum=0,UnwindMaximum=0)
    Reach=(bEnabled=false,ReachTarget=(X=0,Y=0,Z=0),ReachAxis=(X=-1,Y=0,Z=0),ReachMinimum=0.1,ReachMaximum=1,ReachEase=SineEaseInOut)
    Speed=(X=-2,Y=0,Z=0)
    Wave=(bEnabled=true,WaveFrequency=(X=14,Y=0,Z=0),WaveAmplitude=(X=7,Y=0,Z=0),WaveOffset=(X=0,Y=0,Z=0),WaveNoise=(X=0,Y=0,Z=0),WaveMinimum=-1.5,WaveMaximum=1,WaveEase=SineEaseInOut)
    WaveCurve="(EditorCurveData=(Keys=((InterpMode=RCIM_Linear,TangentMode=RCTM_Auto,TangentWeightMode=RCTWM_WeightedNone,Time=0,Value=0.2,ArriveTangent=0,ArriveTangentWeight=0,LeaveTangent=0,LeaveTangentWeight=0),(InterpMode=RCIM_Linear,TangentMode=RCTM_Auto,TangentWeig ...(truncated)"
    bDrawDebug=True
    ExecutePin=@RigUnit_BeginExecution.ExecutePin
   ) {is=struct RigVMUnitNode}
   > &RigUnit_ChainHarmonicsPerItem(
    ChainRoot=(Type=Bone,Name="dyn_tentacle_A_base")
    DrawWorldOffset=(Rotation=(X=0,Y=0,Z=0,W=1),Translation=(X=0,Y=0,Z=0),Scale3D=(X=1,Y=1,Z=1))
    Pendulum=(bEnabled=false,PendulumStiffness=50,PendulumGravity=(X=0,Y=0,Z=0),PendulumBlend=1,PendulumDrag=20,PendulumMinimum=0,PendulumMaximum=0.8,PendulumEase=SineEaseInOut,UnwindAxis=(X=0,Y=0,Z=0),UnwindMinimum=0,UnwindMaximum=0)
    Reach=(bEnabled=true,ReachTarget=(X=5,Y=0,Z=0),ReachAxis=(X=-1,Y=0,Z=0.2),ReachMinimum=0,ReachMaximum=10,ReachEase=SineEaseInOut)
    Speed=(X=0,Y=-4,Z=0)
    Wave=(bEnabled=True,WaveFrequency=(X=0,Y=7,Z=0),WaveAmplitude=(X=0,Y=7,Z=0),WaveOffset=(X=0,Y=0,Z=0),WaveNoise=(X=0,Y=0,Z=0),WaveMinimum=0.1,WaveMaximum=1,WaveEase=SineEaseInOut)
    WaveCurve="(EditorCurveData=(Keys=((InterpMode=RCIM_Linear,TangentMode=RCTM_Auto,TangentWeightMode=RCTWM_WeightedNone,Time=0,Value=0.2,ArriveTangent=0,ArriveTangentWeight=0,LeaveTangent=0,LeaveTangentWeight=0),(InterpMode=RCIM_Linear,TangentMode=RCTM_Auto,TangentWeig ...(truncated)"
    bDrawDebug=True
   )
   > &RigUnit_ChainHarmonicsPerItem(
    ChainRoot=(Type=Bone,Name="dyn_tentacle_B_base")
    DrawWorldOffset=(Rotation=(X=0,Y=0,Z=0,W=1),Translation=(X=0,Y=0,Z=0),Scale3D=(X=1,Y=1,Z=1))
    Pendulum=(bEnabled=false,PendulumStiffness=50,PendulumGravity=(X=-900,Y=-500,Z=0),PendulumBlend=1,PendulumDrag=0.98,PendulumMinimum=0,PendulumMaximum=0.9,PendulumEase=SineEaseInOut,UnwindAxis=(X=0,Y=0,Z=0),UnwindMinimum=0,UnwindMaximum=0)
    Reach=(bEnabled=false,ReachTarget=(X=0,Y=0,Z=0),ReachAxis=(X=-1,Y=0,Z=0),ReachMinimum=0.1,ReachMaximum=1,ReachEase=SineEaseInOut)
    Speed=(X=-2,Y=0,Z=0)
    Wave=(bEnabled=true,WaveFrequency=(X=14,Y=0,Z=0),WaveAmplitude=(X=7,Y=0,Z=0),WaveOffset=(X=1,Y=0,Z=0),WaveNoise=(X=0,Y=0,Z=0),WaveMinimum=-1.5,WaveMaximum=1,WaveEase=SineEaseInOut)
    WaveCurve="(EditorCurveData=(Keys=((InterpMode=RCIM_Linear,TangentMode=RCTM_Auto,TangentWeightMode=RCTWM_WeightedNone,Time=0,Value=0.2,ArriveTangent=0,ArriveTangentWeight=0,LeaveTangent=0,LeaveTangentWeight=0),(InterpMode=RCIM_Linear,TangentMode=RCTM_Auto,TangentWeig ...(truncated)"
    bDrawDebug=True
   )
   > &RigUnit_ChainHarmonicsPerItem(
    ChainRoot=(Type=Bone,Name="dyn_tentacle_B_base")
    DrawWorldOffset=(Rotation=(X=0,Y=0,Z=0,W=1),Translation=(X=0,Y=0,Z=0),Scale3D=(X=1,Y=1,Z=1))
    Pendulum=(bEnabled=false,PendulumStiffness=50,PendulumGravity=(X=0,Y=0,Z=0),PendulumBlend=1,PendulumDrag=20,PendulumMinimum=0,PendulumMaximum=0.8,PendulumEase=SineEaseInOut,UnwindAxis=(X=0,Y=0,Z=0),UnwindMinimum=0,UnwindMaximum=0)
    Reach=(bEnabled=true,ReachTarget=(X=5,Y=0,Z=0),ReachAxis=(X=-1,Y=0,Z=0.2),ReachMinimum=0,ReachMaximum=10,ReachEase=SineEaseInOut)
    Speed=(X=0,Y=-4,Z=0)
    Wave=(bEnabled=True,WaveFrequency=(X=0,Y=7,Z=0),WaveAmplitude=(X=0,Y=7,Z=0),WaveOffset=(X=1,Y=1,Z=0),WaveNoise=(X=0,Y=0,Z=0),WaveMinimum=0.1,WaveMaximum=1,WaveEase=SineEaseInOut)
    WaveCurve="(EditorCurveData=(Keys=((InterpMode=RCIM_Linear,TangentMode=RCTM_Auto,TangentWeightMode=RCTWM_WeightedNone,Time=0,Value=0.2,ArriveTangent=0,ArriveTangentWeight=0,LeaveTangent=0,LeaveTangentWeight=0),(InterpMode=RCIM_Linear,TangentMode=RCTM_Auto,TangentWeig ...(truncated)"
    bDrawDebug=True
   )
   > &RigUnit_ChainHarmonicsPerItem(
    ChainRoot=(Type=Bone,Name="dyn_tentacle_C_base")
    DrawWorldOffset=(Rotation=(X=0,Y=0,Z=0,W=1),Translation=(X=0,Y=0,Z=0),Scale3D=(X=1,Y=1,Z=1))
    Pendulum=(bEnabled=false,PendulumStiffness=50,PendulumGravity=(X=-900,Y=-500,Z=0),PendulumBlend=1,PendulumDrag=0.98,PendulumMinimum=0,PendulumMaximum=0.9,PendulumEase=SineEaseInOut,UnwindAxis=(X=0,Y=0,Z=0),UnwindMinimum=0,UnwindMaximum=0)
    Reach=(bEnabled=false,ReachTarget=(X=0,Y=0,Z=0),ReachAxis=(X=-1,Y=0,Z=0),ReachMinimum=0.1,ReachMaximum=1,ReachEase=SineEaseInOut)
    Speed=(X=-2,Y=0,Z=0)
    Wave=(bEnabled=true,WaveFrequency=(X=14,Y=0,Z=0),WaveAmplitude=(X=7,Y=0,Z=0),WaveOffset=(X=3,Y=0,Z=0),WaveNoise=(X=0,Y=0,Z=0),WaveMinimum=-1.5,WaveMaximum=1,WaveEase=SineEaseInOut)
    WaveCurve="(EditorCurveData=(Keys=((InterpMode=RCIM_Linear,TangentMode=RCTM_Auto,TangentWeightMode=RCTWM_WeightedNone,Time=0,Value=0.2,ArriveTangent=0,ArriveTangentWeight=0,LeaveTangent=0,LeaveTangentWeight=0),(InterpMode=RCIM_Linear,TangentMode=RCTM_Auto,TangentWeig ...(truncated)"
    bDrawDebug=True
   )
   > &RigUnit_ChainHarmonicsPerItem(
    ChainRoot=(Type=Bone,Name="dyn_tentacle_C_base")
    DrawWorldOffset=(Rotation=(X=0,Y=0,Z=0,W=1),Translation=(X=0,Y=0,Z=0),Scale3D=(X=1,Y=1,Z=1))
    Pendulum=(bEnabled=false,PendulumStiffness=100,PendulumGravity=(X=0,Y=0,Z=0),PendulumBlend=1,PendulumDrag=20,PendulumMinimum=0,PendulumMaximum=0.8,PendulumEase=SineEaseInOut,UnwindAxis=(X=0,Y=0,Z=0),UnwindMinimum=0,UnwindMaximum=0)
    Reach=(bEnabled=true,ReachTarget=(X=5,Y=0,Z=0),ReachAxis=(X=-1,Y=0,Z=0.2),ReachMinimum=0,ReachMaximum=10,ReachEase=SineEaseInOut)
    Speed=(X=0,Y=-4,Z=0)
    Wave=(bEnabled=True,WaveFrequency=(X=0,Y=7,Z=0),WaveAmplitude=(X=0,Y=4,Z=0),WaveOffset=(X=3,Y=3,Z=0),WaveNoise=(X=0,Y=0,Z=0),WaveMinimum=0.1,WaveMaximum=1,WaveEase=SineEaseInOut)
    WaveCurve="(EditorCurveData=(Keys=((InterpMode=RCIM_Linear,TangentMode=RCTM_Auto,TangentWeightMode=RCTWM_WeightedNone,Time=0,Value=0.2,ArriveTangent=0,ArriveTangentWeight=0,LeaveTangent=0,LeaveTangentWeight=0),(InterpMode=RCIM_Linear,TangentMode=RCTM_Auto,TangentWeig ...(truncated)"
    bDrawDebug=True
   )
   > &RigUnit_ChainHarmonicsPerItem(
    ChainRoot=(Type=Bone,Name="dyn_tentacle_D_base")
    DrawWorldOffset=(Rotation=(X=0,Y=0,Z=0,W=1),Translation=(X=0,Y=0,Z=0),Scale3D=(X=1,Y=1,Z=1))
    Pendulum=(bEnabled=false,PendulumStiffness=50,PendulumGravity=(X=-900,Y=-500,Z=0),PendulumBlend=1,PendulumDrag=0.98,PendulumMinimum=0,PendulumMaximum=0.9,PendulumEase=SineEaseInOut,UnwindAxis=(X=0,Y=0,Z=0),UnwindMinimum=0,UnwindMaximum=0)
    Reach=(bEnabled=false,ReachTarget=(X=0,Y=0,Z=0),ReachAxis=(X=-1,Y=0,Z=0),ReachMinimum=0.1,ReachMaximum=1,ReachEase=SineEaseInOut)
    Speed=(X=-2,Y=0,Z=0)
    Wave=(bEnabled=true,WaveFrequency=(X=14,Y=0,Z=0),WaveAmplitude=(X=7,Y=0,Z=0),WaveOffset=(X=0,Y=4,Z=0),WaveNoise=(X=0,Y=0,Z=0),WaveMinimum=-1.5,WaveMaximum=1,WaveEase=SineEaseInOut)
    WaveCurve="(EditorCurveData=(Keys=((InterpMode=RCIM_Linear,TangentMode=RCTM_Auto,TangentWeightMode=RCTWM_WeightedNone,Time=0,Value=0.2,ArriveTangent=0,ArriveTangentWeight=0,LeaveTangent=0,LeaveTangentWeight=0),(InterpMode=RCIM_Linear,TangentMode=RCTM_Auto,TangentWeig ...(truncated)"
    bDrawDebug=True
   )
   > &RigUnit_ChainHarmonicsPerItem(
    ChainRoot=(Type=Bone,Name="dyn_tentacle_D_base")
    DrawWorldOffset=(Rotation=(X=0,Y=0,Z=0,W=1),Translation=(X=0,Y=0,Z=0),Scale3D=(X=1,Y=1,Z=1))
    Pendulum=(bEnabled=false,PendulumStiffness=50,PendulumGravity=(X=0,Y=0,Z=0),PendulumBlend=1,PendulumDrag=20,PendulumMinimum=0,PendulumMaximum=0.8,PendulumEase=SineEaseInOut,UnwindAxis=(X=0,Y=0,Z=0),UnwindMinimum=0,UnwindMaximum=0)
    Reach=(bEnabled=true,ReachTarget=(X=5,Y=0,Z=0),ReachAxis=(X=-1,Y=0,Z=0.2),ReachMinimum=0,ReachMaximum=10,ReachEase=SineEaseInOut)
    Speed=(X=0,Y=-4,Z=0)
    Wave=(bEnabled=True,WaveFrequency=(X=0,Y=7,Z=0),WaveAmplitude=(X=0,Y=4,Z=0),WaveOffset=(X=0,Y=4,Z=0),WaveNoise=(X=0,Y=0,Z=0),WaveMinimum=0.1,WaveMaximum=1,WaveEase=SineEaseInOut)
    WaveCurve="(EditorCurveData=(Keys=((InterpMode=RCIM_Linear,TangentMode=RCTM_Auto,TangentWeightMode=RCTWM_WeightedNone,Time=0,Value=0.2,ArriveTangent=0,ArriveTangentWeight=0,LeaveTangent=0,LeaveTangentWeight=0),(InterpMode=RCIM_Linear,TangentMode=RCTM_Auto,TangentWeig ...(truncated)"
    bDrawDebug=True
   )
   &RigUnit_ChainHarmonicsPerItem(
    ChainRoot=(Type=Bone,Name="dyn_tentacle_A_base")
    DrawWorldOffset=(Rotation=(X=0,Y=0,Z=0,W=1),Translation=(X=0,Y=0,Z=0),Scale3D=(X=1,Y=1,Z=1))
    ExecutePin=()
    Pendulum=(bEnabled=false,PendulumStiffness=50,PendulumGravity=(X=-900,Y=-500,Z=0),PendulumBlend=1,PendulumDrag=0.98,PendulumMinimum=0,PendulumMaximum=0.9,PendulumEase=SineEaseInOut,UnwindAxis=(X=0,Y=0,Z=0),UnwindMinimum=0,UnwindMaximum=0)
    Reach=(bEnabled=false,ReachTarget=(X=0,Y=0,Z=0),ReachAxis=(X=1,Y=1,Z=1),ReachMinimum=0,ReachMaximum=0,ReachEase=SineEaseInOut)
    Speed=(X=2,Y=0,Z=-5)
    Wave=(bEnabled=True,WaveFrequency=(X=1,Y=0,Z=10),WaveAmplitude=(X=3,Y=0,Z=0.2),WaveOffset=(X=0,Y=0,Z=0),WaveNoise=(X=0,Y=0,Z=0),WaveMinimum=-1,WaveMaximum=1,WaveEase=SineEaseInOut)
    WaveCurve="(EditorCurveData=(Keys=((InterpMode=RCIM_Linear,TangentMode=RCTM_Auto,TangentWeightMode=RCTWM_WeightedNone,Time=0,Value=0.2,ArriveTangent=0,ArriveTangentWeight=0,LeaveTangent=0,LeaveTangentWeight=0),(InterpMode=RCIM_Linear,TangentMode=RCTM_Auto,TangentWeig ...(truncated)"
    bDrawDebug=True
   )
   &RigUnit_ChainHarmonicsPerItem(
    ChainRoot=(Type=Bone,Name="dyn_tentacle_A_base")
    DrawWorldOffset=(Rotation=(X=0,Y=0,Z=0,W=1),Translation=(X=0,Y=0,Z=0),Scale3D=(X=1,Y=1,Z=1))
    ExecutePin=()
    Pendulum=(bEnabled=false,PendulumStiffness=50,PendulumGravity=(X=-900,Y=-500,Z=0),PendulumBlend=1,PendulumDrag=20,PendulumMinimum=0,PendulumMaximum=0.8,PendulumEase=SineEaseInOut,UnwindAxis=(X=0,Y=0,Z=0),UnwindMinimum=0,UnwindMaximum=0)
    Reach=(bEnabled=true,ReachTarget=(X=5,Y=0,Z=0),ReachAxis=(X=-1,Y=0,Z=0.2),ReachMinimum=0,ReachMaximum=10,ReachEase=SineEaseInOut)
    Speed=(X=-4,Y=-4,Z=-4)
    Wave=(bEnabled=True,WaveFrequency=(X=30,Y=15,Z=30),WaveAmplitude=(X=5,Y=5,Z=5),WaveOffset=(X=0,Y=0,Z=0),WaveNoise=(X=0,Y=0,Z=0),WaveMinimum=0.1,WaveMaximum=1,WaveEase=SineEaseInOut)
    WaveCurve="(EditorCurveData=(Keys=((InterpMode=RCIM_Linear,TangentMode=RCTM_Auto,TangentWeightMode=RCTWM_WeightedNone,Time=0,Value=0.2,ArriveTangent=0,ArriveTangentWeight=0,LeaveTangent=0,LeaveTangentWeight=0),(InterpMode=RCIM_Linear,TangentMode=RCTM_Auto,TangentWeig ...(truncated)"
    bDrawDebug=True
   )
  ]}
]
hierarchy [
 b_AR_Root {loc=<0 0 0>, rot=<0 0 0 1>}
 Melee_TwoHanded_Handle {parent=b_AR_Root, loc=<0 0 0>, rot=<0 0 0.7071 -0.7071>}
 dyn_center_base {parent=Melee_TwoHanded_Handle, loc=<0 0 13.3487>, rot=<0.5059 0.4977 0.5026 -0.4937>}
 dyn_tentacle_A_base {parent=dyn_center_base, loc=<13.6146 4.6258 5.3825>, rot=<0.4542 -0.0126 0.0088 -0.8908>}
 dyn_tentacle_A_01 {parent=dyn_tentacle_A_base, loc=<0 0 0>, rot=<0 -0.0005 -0.0003 -1>}
 dyn_tentacle_A_02 {parent=dyn_tentacle_A_01, loc=<9.5343 0 0>, rot=<0.0288 -0.038 -0.0056 -0.9988>}
 dyn_tentacle_A_03 {parent=dyn_tentacle_A_02, loc=<9.5941 0 0>, rot=<-0.0075 0.0105 -0.0018 -0.9999>}
 dyn_tentacle_A_04 {parent=dyn_tentacle_A_03, loc=<10.035 0 0>, rot=<0 0 0 1>}
 dyn_tentacle_B_base {parent=dyn_center_base, loc=<13.5369 4.8098 -5.4174>, rot=<0.8639 -0.0118 0.0076 -0.5034>}
 dyn_tentacle_B_01 {parent=dyn_tentacle_B_base, loc=<0 0 0>, rot=<0.0004 0.0002 -0.0004 -1>}
 dyn_tentacle_B_02 {parent=dyn_tentacle_B_01, loc=<9.5343 0 0>, rot=<-0.0101 -0.0375 0.0077 -0.9992>}
 dyn_tentacle_B_03 {parent=dyn_tentacle_B_02, loc=<9.594 0 0>, rot=<0.0266 0.0105 0.0016 -0.9996>}
 dyn_tentacle_B_04 {parent=dyn_tentacle_B_03, loc=<10.0351 0 0>, rot=<0 0 0 1>}
 dyn_tentacle_C_base {parent=dyn_center_base, loc=<13.543 -4.6047 -5.5779>, rot=<0.8934 0.0126 0.0065 0.449>}
 dyn_tentacle_C_01 {parent=dyn_tentacle_C_base, loc=<0 0 0>, rot=<0.0007 -0.0006 -0.0002 -1>}
 dyn_tentacle_C_02 {parent=dyn_tentacle_C_01, loc=<9.5343 0 0>, rot=<0.0019 -0.0382 -0.0041 -0.9993>}
 dyn_tentacle_C_03 {parent=dyn_tentacle_C_02, loc=<9.5941 0 0>, rot=<-0.0182 0.0103 -0.0026 -0.9998>}
 dyn_tentacle_C_04 {parent=dyn_tentacle_C_03, loc=<10.035 0 0>, rot=<0 0 0 1>}
 dyn_tentacle_D_base {parent=dyn_center_base, loc=<13.6208 -4.7887 5.2221>, rot=<-0.4695 -0.0129 -0.0096 -0.8828>}
 dyn_tentacle_D_01 {parent=dyn_tentacle_D_base, loc=<0 0 0>, rot=<0.0002 -0.0004 0.0002 -1>}
 dyn_tentacle_D_02 {parent=dyn_tentacle_D_01, loc=<9.5343 0 0>, rot=<0.0316 -0.0383 0.0032 -0.9988>}
 dyn_tentacle_D_03 {parent=dyn_tentacle_D_02, loc=<9.5941 0 0>, rot=<-0.0104 0.0103 0.0029 -0.9999>}
 dyn_tentacle_D_04 {parent=dyn_tentacle_D_03, loc=<10.035 0 0>, rot=<0 0 0 1>}
]
