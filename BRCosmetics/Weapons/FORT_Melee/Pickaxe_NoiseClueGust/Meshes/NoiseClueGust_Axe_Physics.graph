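# Physics asset for the NoiseClueGust pickaxe mesh, giving capsule bodies and joints to the handle and its four dangling tentacles.
# A center base body hangs off the two-handed handle with a tight swing limit.
# Tentacles A through D each chain five tapering capsules from that base, with wider swing limits and narrow twist, so they can bend without spinning.
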PhysicsAsset /BRCosmetics/Weapons/FORT_Melee/Pickaxe_NoiseClueGust/Meshes/NoiseClueGust_Axe_Physics
bodies {
 Melee_TwoHanded_Handle [{shapeType=Capsule, length=10.1516, radius=8.4944, shapeName=Melee_TwoHanded_Handle_capsule}]
 dyn_center_base [{shapeType=Capsule, length=4, radius=9.5, shapeName=dyn_center_base_capsule, center=<2.4074 0 0>, rotation=<rot 90 0.0003 0>}]
 dyn_tentacle_A_01 [{shapeType=Capsule, length=3, radius=5, shapeName=dyn_tentacle_A_01_capsule, rotation=<rot 90 0.0002 0>}]
 dyn_tentacle_A_02 [{shapeType=Capsule, length=4, radius=4, shapeName=dyn_tentacle_A_02_capsule, rotation=<rot 90 0.0001 0>}]
 dyn_tentacle_A_03 [{shapeType=Capsule, length=5, radius=3, shapeName=dyn_tentacle_A_03_capsule, rotation=<rot 90 0.0001 0>}]
 dyn_tentacle_A_04 [{shapeType=Capsule, length=5, radius=3, shapeName=dyn_tentacle_A_04_capsule, rotation=<rot 90 0.0001 0>}]
 dyn_tentacle_A_base [{shapeType=Capsule, length=5.5, radius=5, shapeName=dyn_tentacle_A_base_capsule, rotation=<rot 90 0.0002 0>}]
 dyn_tentacle_B_01 [{shapeType=Capsule, length=6, radius=5, shapeName=dyn_tentacle_B_01_capsule, rotation=<rot 90 -0.0002 0>}]
 dyn_tentacle_B_02 [{shapeType=Capsule, length=4, radius=4, shapeName=dyn_tentacle_B_02_capsule, rotation=<rot 90 -0.0001 0>}]
 dyn_tentacle_B_03 [{shapeType=Capsule, length=5, radius=3, shapeName=dyn_tentacle_B_03_capsule, rotation=<rot 90 -0.0001 0>}]
 dyn_tentacle_B_04 [{shapeType=Capsule, length=5, radius=3, shapeName=dyn_tentacle_B_04_capsule, rotation=<rot 90 -0.0001 0>}]
 dyn_tentacle_B_base [{shapeType=Capsule, length=5.5, radius=5, shapeName=dyn_tentacle_B_base_capsule, rotation=<rot 90 -0.0001 0>}]
 dyn_tentacle_C_01 [{shapeType=Capsule, length=6, radius=5, shapeName=dyn_tentacle_C_01_capsule, rotation=<rot 90 -0.0002 0>}]
 dyn_tentacle_C_02 [{shapeType=Capsule, length=4, radius=4, shapeName=dyn_tentacle_C_02_capsule, rotation=<rot 90 -0.0002 0>}]
 dyn_tentacle_C_03 [{shapeType=Capsule, length=5, radius=3, shapeName=dyn_tentacle_C_03_capsule, rotation=<rot 90 -0.0002 0>}]
 dyn_tentacle_C_04 [{shapeType=Capsule, length=5, radius=3, shapeName=dyn_tentacle_C_04_capsule, rotation=<rot 90 -0.0002 0>}]
 dyn_tentacle_C_base [{shapeType=Capsule, length=5.5, radius=5, shapeName=dyn_tentacle_C_base_capsule, rotation=<rot 90 -0.0002 0>}]
 dyn_tentacle_D_01 [{shapeType=Capsule, length=6, radius=5, shapeName=dyn_tentacle_D_01_capsule, rotation=<rot 90 0.0002 0>}]
 dyn_tentacle_D_02 [{shapeType=Capsule, length=4, radius=4, shapeName=dyn_tentacle_D_02_capsule, rotation=<rot 90 0.0002 0>}]
 dyn_tentacle_D_03 [{shapeType=Capsule, length=5, radius=3, shapeName=dyn_tentacle_D_03_capsule, rotation=<rot 90 0.0002 0>}]
 dyn_tentacle_D_04 [{shapeType=Capsule, length=5, radius=3, shapeName=dyn_tentacle_D_04_capsule, rotation=<rot 90 0.0002 0>}]
 dyn_tentacle_D_base [{shapeType=Capsule, length=5.5, radius=5, shapeName=dyn_tentacle_D_base_capsule, rotation=<rot 90 0.0002 0>}]
}
constraints [
 {bone1Name=dyn_tentacle_A_base, bone2Name=dyn_center_base, swing1LimitDegrees=45, swing2LimitDegrees=45, twistLimitDegrees=5}
 {bone1Name=dyn_tentacle_A_01, bone2Name=dyn_tentacle_A_base, swing1LimitDegrees=45, swing2LimitDegrees=45, twistLimitDegrees=5}
 {bone1Name=dyn_tentacle_A_02, bone2Name=dyn_tentacle_A_01, swing1LimitDegrees=45, swing2LimitDegrees=45, twistLimitDegrees=5}
 {bone1Name=dyn_tentacle_A_03, bone2Name=dyn_tentacle_A_02, swing1LimitDegrees=45, swing2LimitDegrees=45, twistLimitDegrees=5}
 {bone1Name=dyn_tentacle_A_04, bone2Name=dyn_tentacle_A_03, swing1LimitDegrees=45, swing2LimitDegrees=45, twistLimitDegrees=5}
 {bone1Name=dyn_tentacle_B_base, bone2Name=dyn_center_base, swing1LimitDegrees=45, swing2LimitDegrees=45, twistLimitDegrees=5}
 {bone1Name=dyn_tentacle_B_01, bone2Name=dyn_tentacle_B_base, swing1LimitDegrees=45, swing2LimitDegrees=45, twistLimitDegrees=5}
 {bone1Name=dyn_tentacle_B_02, bone2Name=dyn_tentacle_B_01, swing1LimitDegrees=45, swing2LimitDegrees=45, twistLimitDegrees=5}
 {bone1Name=dyn_tentacle_B_03, bone2Name=dyn_tentacle_B_02, swing1LimitDegrees=45, swing2LimitDegrees=45, twistLimitDegrees=5}
 {bone1Name=dyn_tentacle_B_04, bone2Name=dyn_tentacle_B_03, swing1LimitDegrees=45, swing2LimitDegrees=45, twistLimitDegrees=5}
 {bone1Name=dyn_tentacle_C_base, bone2Name=dyn_center_base, swing1LimitDegrees=45, swing2LimitDegrees=45, twistLimitDegrees=5}
 {bone1Name=dyn_tentacle_C_01, bone2Name=dyn_tentacle_C_base, swing1LimitDegrees=45, swing2LimitDegrees=45, twistLimitDegrees=5}
 {bone1Name=dyn_tentacle_C_02, bone2Name=dyn_tentacle_C_01, swing1LimitDegrees=45, swing2LimitDegrees=45, twistLimitDegrees=5}
 {bone1Name=dyn_tentacle_C_03, bone2Name=dyn_tentacle_C_02, swing1LimitDegrees=45, swing2LimitDegrees=45, twistLimitDegrees=5}
 {bone1Name=dyn_tentacle_C_04, bone2Name=dyn_tentacle_C_03, swing1LimitDegrees=45, swing2LimitDegrees=45, twistLimitDegrees=5}
 {bone1Name=dyn_tentacle_D_base, bone2Name=dyn_center_base, swing1LimitDegrees=45, swing2LimitDegrees=45, twistLimitDegrees=5}
 {bone1Name=dyn_tentacle_D_01, bone2Name=dyn_tentacle_D_base, swing1LimitDegrees=45, swing2LimitDegrees=45, twistLimitDegrees=5}
 {bone1Name=dyn_tentacle_D_02, bone2Name=dyn_tentacle_D_01, swing1LimitDegrees=45, swing2LimitDegrees=45, twistLimitDegrees=5}
 {bone1Name=dyn_tentacle_D_03, bone2Name=dyn_tentacle_D_02, swing1LimitDegrees=45, swing2LimitDegrees=45, twistLimitDegrees=5}
 {bone1Name=dyn_tentacle_D_04, bone2Name=dyn_tentacle_D_03, swing1LimitDegrees=45, swing2LimitDegrees=45, twistLimitDegrees=5}
 {bone1Name=dyn_center_base, bone2Name=Melee_TwoHanded_Handle, swing1LimitDegrees=20, swing2LimitDegrees=20, twistLimitDegrees=5}
]
